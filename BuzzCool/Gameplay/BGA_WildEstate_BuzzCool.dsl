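; BuzzCool (WildEstate) item building actor that bursts on spawn and drives the instigator's clone-spawn effect, with a timed lifespan.
; On authority it applies effects to sphere-traced actors, then either removes Recall effects and self-destructs for T1/T2-tagged instigators, or grants a clone-spawn effect (temp A/B and AnimBP tests).
; It attaches to nearby vehicles' closest bone, detaching when they are destroyed.
; A SilverBlazer bubble spawner is defined but uncalled here.
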(event EventBeginPlay
  ;[1 1 1] BeginPlay
  (|CallParentFunction)
  (Utilities|FlowControl|Sequence
    (:then_0
      (Utilities|Time|SetTimerbyEvent (EventDispatchers|CreateEvent :!event Custom|DestroyActor) (ScalableFloat|GetValueatLevel (Variables|Internal|GetSF_MaxLifespanDuration))))
    (:then_1
      (Networking|SwitchHasAuthority
        (:Authority
          (CallFunction|ProcessHitActors)
          ;[1 0.5314 0.0828] Temp for A/B Testing
          (bind _asfort_player_pawn (Utilities|Casting|CastToFortPlayerPawn (Game|GetInstigator)))
          (if (GameplayTags|HasAnyMatchingGameplayTags _asfort_player_pawn (GameplayTags|MakeLiteralGameplayTagContainer "(GameplayTags=((TagName=\"Item.BuzzCool.T1\"),(TagName=\"Item.BuzzCool.T2\")))"))
            (GameplayEffects|RemoveActiveEffectswithGrantedTags (Ability|GetFortAbilitySystemComponent _asfort_player_pawn) "(GameplayTags=((TagName=\"Item.BuzzCool.Recall\")))")
            (Actor|DestroyActor)
            (else
              ;[0.9821 0.1771 1] Temp for AnimBP Testing
              (GameplayEffects|ApplyGameplayEffectToSelf (Ability|GetFortAbilitySystemComponent (Variables|Actor|GetInstigator)) (select (GameplayTags|HasAnyMatchingGameplayTags _asfort_player_pawn (GameplayTags|MakeLiteralGameplayTagContainer "(GameplayTags=((TagName=\"Item.BuzzCool.AnimBP\")))")) "/BuzzCool/Gameplay/Abilities/AnimBP/GE_BuzzCool_CloneSpawn_AnimBP_Item.GE_BuzzCool_CloneSpawn_AnimBP_Item_C" "/BuzzCool/Gameplay/Abilities/GE_BuzzCool_CloneSpawn_Item.GE_BuzzCool_CloneSpawn_Item_C") 0.0 (Class|AthenaFunctionLibrary|MakeSimpleFortEffectContext self)))))))
    (:then_3
      (CallFunction|CheckforVehicleAttach))))

(event Custom|ProcessHitActors
  (Development|PrintString "@@@ ProcessHitActors [BGA_WildEstate_BuzzCool]: CALLED!" false (Variables|Debug|GetbDebugActor))
  (bind _outhits (Collision|MultiSphereTraceForObjects (Transformation|GetActorLocation self) (+ (Transformation|GetActorLocation self) "0.000000,0.000000,1.000000") (ScalableFloat|GetValueatLevel (Variables|Internal|GetSF_ItemExplosion_Radius)) (Variables|Internal|GetItemExplosion_ObjectTypes) false (Utilities|Array|MakeArray self) "ForDuration" true "(R=1.000000,G=0.000000,B=0.000000,A=1.000000)" "(R=0.000000,G=1.000000,B=0.000000,A=1.000000)" 8.0))
  (for _array_element _outhits
    (bind _output_get (Variables|Default|SetCurrentTestActor (.HitActor (Collision|BreakHitResult _array_element))))
    (bind _asfort_player_pawn (Utilities|Casting|CastToFortPlayerPawn _output_get)
      (:then
        (if (and (not (Team|OnSameTeam _asfort_player_pawn (Variables|Actor|GetInstigator))) (!= _asfort_player_pawn (Variables|Actor|GetInstigator)))
          _))
      (:CastFailed
        _))
    (Fortnite|ApplyGameplayEffectsToActor (Variables|Default|GetCurrentTestActor) (Variables|Internal|GetItemExplosion_HitEffects) (Transformation|GetActorLocation self) false 0 (Ability|GetFortAbilitySystemComponent (Variables|Actor|GetInstigator)) (Class|AthenaFunctionLibrary|MakeSimpleFortEffectContext (Variables|Actor|GetInstigator)))))

(event Custom|CheckForVehicleAttach
  (bind _spawnedbga (Variables|Internal|GetSpawnedBGA))
  (bind (_didattach _attachedvehicle) (CallFunction|TryAttachtoVehicle))
  (if _didattach
    (Game|BindEventtoOnDestroyed _attachedvehicle (delegate Custom|OnAttachedVehicleDestroyed))
    (Utilities|IsValid _spawnedbga
      (:Is%20Valid
        (Class|BBGAAthenaSilverBlazerV2|SetIgnoreActor _spawnedbga _attachedvehicle)))))

(event Custom|OnAttachedVehicleDestroyed
  (params (DestroyedActor :type Actor%20Object%20Reference))
  (Transformation|DetachFromActor "KeepWorld" "KeepWorld" "KeepWorld"))

(event Custom|DestroyActor
  (Actor|DestroyActor))

(fn FindAttachableVehicle
  (bind (_outhits _returnvalue) (Collision|MultiSphereTraceForObjects (Transformation|GetActorLocation) (+ (Transformation|GetActorLocation) (* (Math|Vector|SelectVector (Variables|/Script/Engine.MovementComponent|GetVelocity (Variables|Default|GetFortProjectileMovement)) "0, 0,-1.000000" (> (Math|Vector|VectorLength (Variables|/Script/Engine.MovementComponent|GetVelocity (Variables|Default|GetFortProjectileMovement))) 0)) (Variables|Internal|GetVehicleCheckDistance))) 5e+01 (Utilities|Array|MakeArray "ObjectTypeQuery5")))
  (if _returnvalue
    (for _array_element _outhits
      (if (Utilities|DoesObjectImplementInterface (.HitActor (Collision|BreakHitResult _array_element)) "/Script/FortniteGame.FortVehicleInterface")
        (return true (.HitActor (Collision|BreakHitResult _array_element)))
        (else
          (return false 0))))
    (else
      _)))

(fn OnRep_SpawnedBGA
  (Class|BBGAAthenaSilverBlazerV2|SetFXPos (Variables|Internal|GetSpawnedBGA) (Transformation|GetActorLocation)))

(fn SpawnBubbleSafeZone
  (bind _returnvalue_location (Math|Transform|BreakTransform (Transformation|GetActorTransform)))
  ;[0 0.07676 1] SpawnBubbleSafeZone
  ; Spawn the bubble area BGA
  (bind _returnvalue (Game|SpawnBuildingActorfromClass "/Game/Athena/Items/Consumables/SilverBlazer/B_BGA_Athena_SilverBlazer_V2.B_BGA_Athena_SilverBlazer_V2_C" (Math|Transform|MakeTransform _returnvalue_location) 0 false 0.0 (Variables|Missions|GetAssociatedMissionParam) (Game|GetInstigator)))
  (return _returnvalue))

(fn TryAttachToVehicle
  (bind (_canattach _hitactor) (CallFunction|FindAttachableVehicle))
  (if _canattach
    (bind _asfort_vehicle_interface (Utilities|Casting|CastToFortVehicleInterface _hitactor))
    (bind _returnvalue (Transformation|AttachComponentToComponent (Variables|Default|GetStaticMesh) (Class|FortVehicleInterface|GetSkeletalMeshComponent _asfort_vehicle_interface) (.ReturnValue (Components|SkinnedMesh|FindClosestBone (Class|FortVehicleInterface|GetSkeletalMeshComponent _asfort_vehicle_interface) (Transformation|GetActorLocation) 0.0 false)) "KeepWorld" "KeepWorld" "KeepWorld" true))
    (Collision|SetCollisionEnabled (Variables|Default|GetStaticMesh) "QueryOnly")
    (return _returnvalue _hitactor)
    (else
      (return false 0))))

(fn ConstructionScript
  (|CallParentFunction))
