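; BuzzCool processor that runs a falling clone spawn actor, its Living World encounters and its loot drop, mostly server-side.
; Spawns the actor high overhead; landing starts or resumes the active-clone encounter.
; Damaging its donut prop turns clones hostile; destroying it releases that encounter, triggers a clone-spawn encounter at the donut's point providers, and drops loot.
; Processors share loot tier pools over GameplayEventRouter, removing each used tier.
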(event Custom|Test_CloneSpawnActor
  (bind (_returnvalue_x _returnvalue_y _returnvalue_z) (Math|Vector|BreakVector (Transformation|GetActorLocation self)))
  (Networking|SwitchHasAuthority
    (:Authority
      (Development|PrintString "@@@ CloneSpawnActor [BP_BuzzCool_CloneSpawnActor_Processor]: CALLED!" false (Variables|Debug|GetbDebugProcessor))
      (Utilities|FlowControl|Sequence
        (:then_0
          (Variables|Internal|SetbLootInitialized false)
          (bind _returnvalue (Building|SpawnBuildingGameplayActorfromWorldContext "/BuzzCool/Gameplay/BGA_BuzzCool_CloneSpawnActor.BGA_BuzzCool_CloneSpawnActor_C" (Math|Transform|MakeTransform (Math|Vector|MakeVector _returnvalue_x _returnvalue_y (+ _returnvalue_z 7.68e+03)) (Math|Rotator|MakeRotator 0.0 0.0 (Math|Random|RandomFloatinRange 0.0 3.6e+02)))))
          (bind _asbga_buzz_cool_clone_spawn_actor (Utilities|Casting|CastToBGA_BuzzCool_CloneSpawnActor _returnvalue))
          (bind _output_get (Variables|Internal|SetCloneSpawnActor _asbga_buzz_cool_clone_spawn_actor))
          (bind _output_get_2 (Variables|Internal|SetDonutProp (Variables|/SonicSpearGameplay/Gameplay/Drops/BGA_SonicSpear_MainDropParent.BGA_SonicSpear_MainDropParent_C|GetSpawnedProp _output_get)))
          (EventDispatchers|BindEventtoOnDied _output_get_2 (EventDispatchers|CreateEvent :!event Custom|OnSpawnActorDied) :!class /Script/FortniteGame.BuildingActor)
          (EventDispatchers|BindEventtoOnDamaged _output_get_2 (EventDispatchers|CreateEvent :!event Custom|OnSpawnActorDamaged) :!class /Script/FortniteGame.BuildingActor)
          (Default|BindEventtoOnLanded (Variables|Internal|GetCloneSpawnActor) (EventDispatchers|CreateEvent :!event Custom|OnSpawnActorLanded) :!class /SonicSpearGameplay/Gameplay/Drops/BGA_SonicSpear_MainDropParent.BGA_SonicSpear_MainDropParent_C)
          (Variables|Internal|SetbClonesAreAggro false)
          (bind _asbp_buzz_cool_donut_large (Utilities|Casting|CastToBP_BuzzCool_Donut_Large (Variables|Internal|GetDonutProp)))
          (Variables|/BuzzCool/Gameplay/BP_BuzzCool_Donut_Large.BP_BuzzCool_Donut_Large_C|SetCloneSpawnActorProcessor self _asbp_buzz_cool_donut_large))
        (:then_1
          (Class|FortAsyncActionCurrentPlaylistReady|PlaylistDataReadyAsync)
          (Utilities|FlowControl|Sequence
            (:then_0
              (bind (_asynctaskproxy _payload) (GameplayEventRouter|StartListeningToEvent (GameplayEventRouter|GetGlobalEventRouter) "/BuzzCool/Gameplay/EventRouter/BuzzCool_CloneSpawnActor_Events.BuzzCool_CloneSpawnActor_Events")
                (:then
                  (Variables|Internal|SetAsyncTaskEventListener _asynctaskproxy))
                (:OnEventReceived
                  (if (not (== (.CallingProcessor_13_91224E9842B8318E588E1B8B38878BC0 (Utilities|Struct|BreakBuzzCoolCloneSpawnActorEvents _payload)) self))
                    (switch Utilities|FlowControl|Switch|SwitchonBuzzCool_CloneSpawnActor_Events_Enum (Utilities|Struct|BreakBuzzCoolCloneSpawnActorEvents _payload)
                      (:NewEnumerator0
                        (CallFunction|RemoveUsedLootTierGroup (.LootTierGroupName_5_E5C8547944DCA720FA31A7AA1A17622B (Utilities|Struct|BreakBuzzCoolCloneSpawnActorEvents _payload))))
                      (:NewEnumerator1
                        (Networking|SwitchHasAuthority
                          (:Authority
                            (GameplayEventRouter|BroadcastEvent (GameplayEventRouter|GetGlobalEventRouter) (Utilities|Struct|MakeBuzzCoolCloneSpawnActorEvents "NewEnumerator2" "None" (Variables|Internal|GetLootTierNameArray)) 0))))
                      (:NewEnumerator2
                        (CallFunction|InitializeLoot (.LootTierNameArray_9_A5FF8710459E8A4492554395ECEF07CE (Utilities|Struct|BreakBuzzCoolCloneSpawnActorEvents _payload)))))))))
            (:then_1
              (Networking|SwitchHasAuthority
                (:Authority
                  (GameplayEventRouter|BroadcastEvent (GameplayEventRouter|GetGlobalEventRouter) (Utilities|Struct|MakeBuzzCoolCloneSpawnActorEvents :EventKind_2_FB4D01D049124EC6A4B4E3933A2C82E2 "NewEnumerator1" :LootTierGroupName_5_E5C8547944DCA720FA31A7AA1A17622B "None" :CallingProcessor_13_91224E9842B8318E588E1B8B38878BC0 self) 0))))))))))

(event EventBeginPlay
  (bind (_returnvalue_x _returnvalue_y _returnvalue_z) (Math|Vector|BreakVector (Transformation|GetActorLocation self)))
  (|CallParentFunction)
  (Networking|SwitchHasAuthority
    (:Authority
      (Development|PrintString "@@@ CloneSpawnActor [BP_BuzzCool_CloneSpawnActor_Processor]: CALLED!" false (Variables|Debug|GetbDebugProcessor))
      (Utilities|FlowControl|Sequence
        (:then_0
          (Variables|Internal|SetbLootInitialized false)
          (bind _returnvalue (Building|SpawnBuildingGameplayActorfromWorldContext "/BuzzCool/Gameplay/BGA_BuzzCool_CloneSpawnActor.BGA_BuzzCool_CloneSpawnActor_C" (Math|Transform|MakeTransform (Math|Vector|MakeVector _returnvalue_x _returnvalue_y (+ _returnvalue_z 7.68e+03)) (Math|Rotator|MakeRotator 0.0 0.0 (Math|Random|RandomFloatinRange 0.0 3.6e+02)))))
          (bind _asbga_buzz_cool_clone_spawn_actor (Utilities|Casting|CastToBGA_BuzzCool_CloneSpawnActor _returnvalue))
          (bind _output_get (Variables|Internal|SetCloneSpawnActor _asbga_buzz_cool_clone_spawn_actor))
          (bind _output_get_2 (Variables|Internal|SetDonutProp (Variables|/SonicSpearGameplay/Gameplay/Drops/BGA_SonicSpear_MainDropParent.BGA_SonicSpear_MainDropParent_C|GetSpawnedProp _output_get)))
          (EventDispatchers|BindEventtoOnDied _output_get_2 (EventDispatchers|CreateEvent :!event Custom|OnSpawnActorDied) :!class /Script/FortniteGame.BuildingActor)
          (EventDispatchers|BindEventtoOnDamaged _output_get_2 (EventDispatchers|CreateEvent :!event Custom|OnSpawnActorDamaged) :!class /Script/FortniteGame.BuildingActor)
          (Default|BindEventtoOnLanded (Variables|Internal|GetCloneSpawnActor) (EventDispatchers|CreateEvent :!event Custom|OnSpawnActorLanded) :!class /SonicSpearGameplay/Gameplay/Drops/BGA_SonicSpear_MainDropParent.BGA_SonicSpear_MainDropParent_C)
          (Variables|Internal|SetbClonesAreAggro false)
          (bind _asbp_buzz_cool_donut_large (Utilities|Casting|CastToBP_BuzzCool_Donut_Large (Variables|Internal|GetDonutProp)))
          (Variables|/BuzzCool/Gameplay/BP_BuzzCool_Donut_Large.BP_BuzzCool_Donut_Large_C|SetCloneSpawnActorProcessor self _asbp_buzz_cool_donut_large))
        (:then_1
          (Class|FortAsyncActionCurrentPlaylistReady|PlaylistDataReadyAsync)
          (Utilities|FlowControl|Sequence
            (:then_0
              (bind (_asynctaskproxy _payload) (GameplayEventRouter|StartListeningToEvent (GameplayEventRouter|GetGlobalEventRouter) "/BuzzCool/Gameplay/EventRouter/BuzzCool_CloneSpawnActor_Events.BuzzCool_CloneSpawnActor_Events")
                (:then
                  (Variables|Internal|SetAsyncTaskEventListener _asynctaskproxy))
                (:OnEventReceived
                  (if (not (== (.CallingProcessor_13_91224E9842B8318E588E1B8B38878BC0 (Utilities|Struct|BreakBuzzCoolCloneSpawnActorEvents _payload)) self))
                    (switch Utilities|FlowControl|Switch|SwitchonBuzzCool_CloneSpawnActor_Events_Enum (Utilities|Struct|BreakBuzzCoolCloneSpawnActorEvents _payload)
                      (:NewEnumerator0
                        (CallFunction|RemoveUsedLootTierGroup (.LootTierGroupName_5_E5C8547944DCA720FA31A7AA1A17622B (Utilities|Struct|BreakBuzzCoolCloneSpawnActorEvents _payload))))
                      (:NewEnumerator1
                        (Networking|SwitchHasAuthority
                          (:Authority
                            (GameplayEventRouter|BroadcastEvent (GameplayEventRouter|GetGlobalEventRouter) (Utilities|Struct|MakeBuzzCoolCloneSpawnActorEvents "NewEnumerator2" "None" (Variables|Internal|GetLootTierNameArray)) 0))))
                      (:NewEnumerator2
                        (CallFunction|InitializeLoot (.LootTierNameArray_9_A5FF8710459E8A4492554395ECEF07CE (Utilities|Struct|BreakBuzzCoolCloneSpawnActorEvents _payload)))))))))
            (:then_1
              (Networking|SwitchHasAuthority
                (:Authority
                  (GameplayEventRouter|BroadcastEvent (GameplayEventRouter|GetGlobalEventRouter) (Utilities|Struct|MakeBuzzCoolCloneSpawnActorEvents :EventKind_2_FB4D01D049124EC6A4B4E3933A2C82E2 "NewEnumerator1" :LootTierGroupName_5_E5C8547944DCA720FA31A7AA1A17622B "None" :CallingProcessor_13_91224E9842B8318E588E1B8B38878BC0 self) 0))))))))))

(event Custom|OnSpawnActorDied
  (params (DamagedActor :type Actor%20Object%20Reference) (Damage :type Float%20%28single-precision%29) (InstigatedBy :type Controller%20Object%20Reference) (DamageCauser :type Actor%20Object%20Reference) (HitLocation :type Vector) (FHitComponent :type Primitive%20Component%20Object%20Reference) (BoneName :type Name) (Momentum :type Vector))
  (bind _encounterref_clonespawns (Variables|Internal|GetEncounterRef_CloneSpawns))
  ;[0.4908 0.0198 1] Clone Spawn Actor tracking...
  (Networking|SwitchHasAuthority
    (:Authority
      (Utilities|FlowControl|Sequence
        (:then_0
          (Development|PrintString (Utilities|String|Append "@@@@@@ OnSpawnActorDied [BP_BuzzCool_CloneSpawnActor_Processor]: CALLED - " (Utilities|GetDisplayName DamagedActor) " DIED!") false (Variables|Debug|GetbDebugProcessor))
          (CallFunction|ReleaseActiveEncounter)
          (CallFunction|SpawnActorCleanup)
          (Utilities|IsValid _encounterref_clonespawns
            (:Is%20Valid
              (Class|FortAthenaLivingWorldEncounterInstance|ResumeEncounter _encounterref_clonespawns)
              (Class|FortAthenaLivingWorldEncounterInstance|RequestEventGeneration _encounterref_clonespawns))
            (:Is%20Not%20Valid
              (CallFunction|EnableCloneSpawnPointProviders true)
              (Utilities|FlowControl|DelayUntilNextTick)
              (bind (_isvalid _returnvalue) (Default|StartActivityEncounterStaticPointProvider (Game|GetGameState) self (Variables|Internal|GetEncounterType_CloneSpawns) (CallFunction|GetCloneSpawnPointProviders)))
              (if _isvalid
                (bind _output_get (Variables|Internal|SetEncounterRef_CloneSpawns _returnvalue))
                (Class|FortAthenaLivingWorldEncounterInstance|RequestEventGeneration _output_get)
                (else
                  (Fortnite|FortShippingLog self "[BP_BuzzCool_CloneSpawnActor_Processor] :: Tried to start an Encounter, but could not! Make sure the Encounter is defined properly on the asset AND in the Living World Manager!!" true))))))
        (:then_1
          (Ability|SendGameplayEventtoActor InstigatedBy "(TagName=\"Event.BuzzCool.SpawnActor.Destroyed\")" (Utilities|Struct|MakeGameplayEventData "(TagName=\"\")" self 0 InstigatedBy))
          (CallFunction|ProcessLootDrop DamagedActor))))))

(event Custom|ProcessLootDrop
  (params (DropActor :type Actor%20Object%20Reference))
  (bind _loottiernamearray (Variables|Internal|GetLootTierNameArray))
  (bind _outitem (Utilities|Array|RandomArrayItem _loottiernamearray))
  (bind _returnvalue (Utilities|Array|IsEmpty _loottiernamearray))
  (bind _defaultloottiername (Variables|Internal|GetDefaultLootTierName))
  (bind _returnvalue_2 (select _returnvalue _defaultloottiername _outitem))
  ;[1 0.1364 0.00615] Clone Spawn Actor Loot Processing...
  (if (not (Variables|Internal|GetbLootInitialized))
    (CallFunction|InitializeLoot))
  (Class|AthenaFunctionLibrary|SpawnLootfromTableUsingItemEntry _returnvalue_2 1 -1 (+ (Transformation|GetActorLocation DropActor) (* (Transformation|GetActorUpVector) 6e+02)) (Transformation|GetActorForwardVector) 0 true true false 0 "Other" "NPCElimination")
  (Networking|SwitchHasAuthority
    (:Authority
      (GameplayEventRouter|BroadcastEvent (GameplayEventRouter|GetGlobalEventRouter) (Utilities|Struct|MakeBuzzCoolCloneSpawnActorEvents "NewEnumerator0" _returnvalue_2) 0))))

(event Custom|OnSpawnActorDamaged
  (params (DamagedActor :type Actor%20Object%20Reference) (Damage :type Float%20%28single-precision%29) (InstigatedBy :type Controller%20Object%20Reference) (DamageCauser :type Actor%20Object%20Reference) (HitLocation :type Vector) (FHitComponent :type Primitive%20Component%20Object%20Reference) (BoneName :type Name) (Momentum :type Vector))
  ;[0.4908 0.0198 1] Clone Spawn Actor tracking...
  (Networking|SwitchHasAuthority
    (:Authority
      (if (not (Variables|Internal|GetbClonesAreAggro))
        (Variables|Internal|SetbClonesAreAggro true)
        (for _array_element (Variables|Internal|GetActiveCloneArray)
          (Utilities|IsValid _array_element
            (:Is%20Valid
              (Affiliation|RemovefromAllFactions _array_element)
              (Affiliation|AddtoFaction _array_element "(TagName=\"Athena.Faction.BuzzCool.Hostile\")"))))))))

(event Custom|OnSpawnActorLanded
  (bind _encounterref_activeclones (Variables|Internal|GetEncounterRef_ActiveClones))
  (Networking|SwitchHasAuthority
    (:Authority
      (Development|PrintString "@@@@@@ OnSpawnActorLanded [BP_BuzzCool_CloneSpawnActor_Processor]: CALLED!" false (Variables|Debug|GetbDebugProcessor))
      (Utilities|IsValid _encounterref_activeclones
        (:Is%20Valid
          (Class|FortAthenaLivingWorldEncounterInstance|ResumeEncounter _encounterref_activeclones))
        (:Is%20Not%20Valid
          (Class|FortAthenaLivingWorldPointProviderInterface|EnablePointProvider_DEPRECATED (Variables|Tunable|GetLivingWorldVol_ActiveClones))
          (Utilities|FlowControl|DelayUntilNextTick)
          (bind (_isvalid _returnvalue) (Default|StartActivityEncounter (Game|GetGameState) (Variables|Tunable|GetLivingWorldVol_ActiveClones) self (Variables|Internal|GetEncounterType_ActiveClones)))
          (bind _output_get (Variables|Internal|SetEncounterRef_ActiveClones _returnvalue))
          (Class|FortAthenaLivingWorldEncounterInstance|RequestEventGeneration _output_get)
          (EventDispatchers|BindEventtoOnEventSpawned _output_get (EventDispatchers|CreateEvent :!event Custom|OnActiveCloneEventSpawned) :!class /Script/LagerRuntime.FortAthenaLivingWorldEncounterInstance)
          (EventDispatchers|BindEventtoOnActorDied _output_get (EventDispatchers|CreateEvent :!event Custom|OnActiveCloneActorDied)))))))

(event Custom|ReleaseActiveEncounter
  (bind _encounterref_activeclones (Variables|Internal|GetEncounterRef_ActiveClones))
  (Networking|SwitchHasAuthority
    (:Authority
      (Utilities|IsValid _encounterref_activeclones
        (:Is%20Valid
          (Class|FortAthenaLivingWorldEncounterInstance|PauseEncounter _encounterref_activeclones)
          (Class|FortAthenaLivingWorldManager|ReleaseEncounter (Class|FortAthenaLivingWorldManager|GetLivingWorldManager) _encounterref_activeclones))))))

(event Custom|OnActiveCloneEventSpawned
  (params (EncounterInstance :type Fort%20Athena%20Living%20World%20Encounter%20Instance%20Object%20Reference) (SpawnedActors :type Array%20of%20Actor%20Object%20References))
  ;[0 0.4844 0.2846] Active Clone Encounter tracking...
  (for _array_element SpawnedActors
    (Utilities|IsValid _array_element
      (:Is%20Valid
        (bind _aspawn_wild_estate_buzz_cool (Utilities|Casting|CastToPawn_WildEstate_BuzzCool _array_element))
        (Development|PrintString "@@@@@@ OnActiveCloneEventSpawned [BP_BuzzCool_CloneSpawnActor_Processor]: CALLED - a Clone is SPAWNED!" false (Variables|Debug|GetbDebugProcessor))
        (Utilities|Array|AddUnique (Variables|Internal|GetActiveCloneArray) _aspawn_wild_estate_buzz_cool)
        (if (Variables|Internal|GetbClonesAreAggro)
          (Affiliation|RemovefromAllFactions _aspawn_wild_estate_buzz_cool)
          (Affiliation|AddtoFaction _aspawn_wild_estate_buzz_cool "(TagName=\"Athena.Faction.BuzzCool.Hostile\")"))))))

(event Custom|Test_EndCloneSpawnEnc
  (bind _encounterref_clonespawns (Variables|Internal|GetEncounterRef_CloneSpawns))
  (Networking|SwitchHasAuthority
    (:Authority
      (Development|PrintString "@@@ Test_EndCloneSpawnEnc [BP_BuzzCool_CloneSpawnActor_Processor]: CALLED!" false (Variables|Debug|GetbDebugProcessor))
      (Class|FortAthenaLivingWorldEncounterInstance|PauseEncounter _encounterref_clonespawns)
      (Class|BPFLLagerCommunFunctions|EndActivityEncounter (Game|GetGameState) _encounterref_clonespawns)
      (Variables|Internal|SetEncounterRef_CloneSpawns 0))))

(event Custom|StopCloneSpawnEnc
  (bind _encounterref_clonespawns (Variables|Internal|GetEncounterRef_CloneSpawns))
  (Networking|SwitchHasAuthority
    (:Authority
      (Development|PrintString "@@@ Test_EndCloneSpawnEnc [BP_BuzzCool_CloneSpawnActor_Processor]: CALLED!" false (Variables|Debug|GetbDebugProcessor))
      (Class|FortAthenaLivingWorldEncounterInstance|PauseEncounter _encounterref_clonespawns)
      (Class|BPFLLagerCommunFunctions|EndActivityEncounter (Game|GetGameState) _encounterref_clonespawns)
      (Variables|Internal|SetEncounterRef_CloneSpawns 0))))

(event Custom|OnActiveCloneActorDied
  (params (EncounterInstance :type Fort%20Athena%20Living%20World%20Encounter%20Instance%20Object%20Reference) (DiedActor :type Actor%20Object%20Reference))
  ;[0 0.4844 0.2846] Active Clone Encounter tracking...
  (Utilities|IsValid DiedActor
    (:Is%20Valid
      (bind _aspawn_wild_estate_buzz_cool (Utilities|Casting|CastToPawn_WildEstate_BuzzCool DiedActor))
      (Development|PrintString "@@@@@@ OnActiveCloneActorDied [BP_BuzzCool_CloneSpawnActor_Processor]: CALLED - a Clone is DESTROYED!" false (Variables|Debug|GetbDebugProcessor))
      (Utilities|Array|RemoveItem (Variables|Internal|GetActiveCloneArray) _aspawn_wild_estate_buzz_cool))))

(event EventAddActiveClone
  (params (NewClone :type Pawn%20Wild%20Estate%20Buzz%20Cool%20Object%20Reference))
  (Utilities|Array|AddUnique (Variables|Internal|GetActiveCloneArray) NewClone)
  (if (Variables|Internal|GetbClonesAreAggro)
    (Affiliation|RemovefromAllFactions NewClone)
    (Affiliation|AddtoFaction NewClone "(TagName=\"Athena.Faction.BuzzCool.Hostile\")")))

(event Custom|Test_RemoveLootTier
  (Networking|SwitchHasAuthority
    (:Authority
      (GameplayEventRouter|BroadcastEvent (GameplayEventRouter|GetGlobalEventRouter) (Utilities|Struct|MakeBuzzCoolCloneSpawnActorEvents "NewEnumerator0" (Utilities|Name|MakeLiteralName "LTG_BuzzCool_Spawn_D")) 0))))

(fn EnableCloneSpawnPointProviders
  (params (bEnable :type Boolean))
  (for _array_element (CallFunction|GetCloneSpawnPointProviders)
    (if bEnable
      (Class|FortAthenaLivingWorldPointProviderInterface|EnablePointProvider_DEPRECATED _array_element)
      (else
        (Class|FortAthenaLivingWorldStaticPointProvider|DisablePointProvider _array_element)))
    (if (Variables|Debug|GetbDebugProcessor)
      (Rendering|Debug|DrawDebugSphere (Transformation|GetActorLocation _array_element) 5e+01 12 "(R=0.980780,G=0.000000,B=1.000000,A=1.000000)" 1e+01 2.0))))

(fn GetActiveClones
  (return (Variables|Internal|GetActiveCloneArray)))

(fn GetCloneSpawnPointProviders
  (bind _asbp_buzz_cool_donut_large (Utilities|Casting|CastToBP_BuzzCool_Donut_Large (Variables|Internal|GetDonutProp)))
  (return (Variables|/BuzzCool/Gameplay/BP_BuzzCool_Donut_Large.BP_BuzzCool_Donut_Large_C|GetLivingWorld_SpawnLocations _asbp_buzz_cool_donut_large)))

(fn InitializeLoot
  (params (SentLootArray :type Array%20of%20Names))
  (bind _returnvalue (Utilities|Name|MakeLiteralName "LTG_BuzzCool_Spawn_A"))
  (if (not (Variables|Internal|GetbLootInitialized))
    (Development|PrintString "@@@@@@ InitializeLoot [BP_BuzzCool_CloneSpawnActor_Processor]: CALLED! " false (Variables|Debug|GetbDebugProcessor))
    (Variables|Internal|SetbLootInitialized true)
    (if (Utilities|Array|IsEmpty (Variables|GetSentLootArray))
      (Variables|Internal|SetDefaultLootTierName _returnvalue)
      (bind _output_get (Variables|Internal|SetLootTierNameArray (Utilities|Array|MakeArray _returnvalue (Utilities|Name|MakeLiteralName "LTG_BuzzCool_Spawn_B") (Utilities|Name|MakeLiteralName "LTG_BuzzCool_Spawn_C") (Utilities|Name|MakeLiteralName "LTG_BuzzCool_Spawn_D"))))
      (Utilities|Array|Shuffle _output_get)
      (else
        (Variables|Internal|SetDefaultLootTierName (Utilities|Array|RandomArrayItem (Variables|GetSentLootArray)))
        (Variables|Internal|SetLootTierNameArray (Variables|GetSentLootArray))))
    (if (Variables|Debug|GetbDebugProcessor)
      (for (_array_element _array_index) (Variables|Internal|GetLootTierNameArray)
        (Development|PrintString (Utilities|String|Append "@@@@@@@@@ InitializeLoot [BP_BuzzCool_CloneSpawnActor_Processor]: RESULTS: " "[" (Utilities|String|ToString(Integer) _array_index) "] = " (Utilities|String|ToString(Name) _array_element)) false (Variables|Debug|GetbDebugProcessor))))))

(fn RemoveUsedLootTierGroup
  (params (UsedLootTierGroup :type Name))
  (bind _loottiernamearray (Variables|Internal|GetLootTierNameArray))
  (Development|PrintString (Utilities|String|Append "@@@@@@ RemoveUsedLootTierGroup [BP_BuzzCool_CloneSpawnActor_Processor]: CALLED, REMOVING  " (Utilities|String|ToString(Name) (Variables|GetUsedLootTierGroup)) "!") false (Variables|Debug|GetbDebugProcessor))
  (if (not (Variables|Internal|GetbLootInitialized))
    (CallFunction|InitializeLoot))
  (Utilities|Array|RemoveItem _loottiernamearray (Variables|GetUsedLootTierGroup))
  (Variables|Internal|SetDefaultLootTierName (Utilities|Array|RandomArrayItem _loottiernamearray))
  (if (Variables|Debug|GetbDebugProcessor)
    (for (_array_element _array_index) (Variables|Internal|GetLootTierNameArray)
      (Development|PrintString (Utilities|String|Append "@@@@@@@@@ RemoveUsedLootTierGroup [BP_BuzzCool_CloneSpawnActor_Processor]: RESULTS: " "[" (Utilities|String|ToString(Integer) _array_index) "] = " (Utilities|String|ToString(Name) _array_element)) false (Variables|Debug|GetbDebugProcessor)))))

(fn SpawnActorCleanup
  (bind _obstaclevol_clonespawnactor (Variables|Tunable|GetObstacleVol_CloneSpawnActor))
  (Utilities|IsValid _obstaclevol_clonespawnactor
    (:Is%20Valid
      (Actor|DestroyActor _obstaclevol_clonespawnactor))))

(fn ConstructionScript)
