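; BuzzCool clone spawn point provider that configures each spawned wild estate BuzzCool clone pawn as a hostile.
; On BeginPlay with authority, it binds a handler to the actor-spawned dispatcher.
; The handler casts the spawn to the clone pawn, sets itself as its spawn provider, and moves it into the BuzzCool hostile faction.
; It then applies a launch gameplay effect through the spawn's ability system, with optional debug printing.
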(event EventBeginPlay
  (Networking|SwitchHasAuthority
    (:Authority
      (EventDispatchers|BindEventtoOnActorSpawned (delegate Custom|OnActorSpawned_Event)))))

(event Custom|OnActorSpawned_Event
  (params (SpawnedActor :type Actor%20Object%20Reference))
  (bind _aspawn_wild_estate_buzz_cool (Utilities|Casting|CastToPawn_WildEstate_BuzzCool SpawnedActor))
  (Variables|/BuzzCool/NPCs/BuzzCool_Clones/Pawn/Pawn_WildEstate_BuzzCool.Pawn_WildEstate_BuzzCool_C|SetSpawnProviderRef self _aspawn_wild_estate_buzz_cool)
  (Development|PrintString (Utilities|String|Append "@@@@@@ OnActorSpawned_Event [BP_BuzzCool_CloneSpawnActor_StaticPointProvider]: CALLED - " (Utilities|GetDisplayName SpawnedActor) " SPAWNED!") false (Variables|Debug|GetbDebugPointProvider))
  (Affiliation|RemovefromAllFactions _aspawn_wild_estate_buzz_cool)
  (Affiliation|AddtoFaction _aspawn_wild_estate_buzz_cool "(TagName=\"Athena.Faction.BuzzCool.Hostile\")")
  (GameplayEffects|ApplyGameplayEffectToSelf :self (Ability|GetFortAbilitySystemComponent SpawnedActor) :GameplayEffectClass "/BuzzCool/Gameplay/Abilities/GE_BuzzCool_CloneSpawnActor_Launch.GE_BuzzCool_CloneSpawnActor_Launch_C" :Level 0.0))

(fn ConstructionScript)
