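; BP_BuzzCool_Donut_AimAssist is a BuzzCool gameplay Blueprint that makes a donut object an aim-assist target.
; It registers its BodyCollisionBox with the aim-assist system on BeginPlay and unregisters on EndPlay.
; Its target query functions accept auto-fire from any attacker, while both deployed-target auto-fire checks, for allies and for self, return false.
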(event EventBeginPlay
  (AimAssist|RegisterAimAssistTarget 0 (Variables|Default|GetBodyCollisionBox) false))

(event EventEndPlay
  (params (EndPlayReason :type EEndPlayReason%20Enum :default "Destroyed"))
  (AimAssist|UnregisterAimAssistTarget 0))

(fn GetTargetOwner
  (return 0))

(fn IsValidAutoFireTarget
  (params (Attacker :type Actor%20Object%20Reference))
  (return true))

(fn ShouldAlliesAutoFireOnDeployedTarget
  (return false))

(fn ShouldAutoFireOnSelfDeployedTarget
  (return false))

(fn ConstructionScript)
