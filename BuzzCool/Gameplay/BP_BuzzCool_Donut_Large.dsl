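; Large BuzzCool donut objective actor that rides a SonicSpear drop and can be destroyed.
; When the owning drop lands, the server spawns static point-provider actors at relative transforms, optionally attaches them, and destroys them on EndPlay.
; Local players entering hotfixable-radius spheres broadcast gameplay events for auto-ping and the "Destroy" marker.
; Death plays frosting-colored Niagara and sound; a cheat-triggered timer draws provider debug arrows.
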(event EventBeginPlay
  (bind _returnvalue (Actor|GetOwner))
  ;[0 1 0 0.5] [BeginPlay]
  (|CallParentFunction)
  (Utilities|FlowControl|Sequence
    (:then_0
      (Networking|SwitchHasAuthority
        (:Authority
          (Utilities|IsValid _returnvalue
            (:Is%20Valid
              (bind _asbga_sonic_spear_main_drop_parent (Utilities|Casting|CastToBGA_SonicSpear_MainDropParent _returnvalue)
                (:then
                  (Default|BindEventtoOnLanded _asbga_sonic_spear_main_drop_parent (EventDispatchers|CreateEvent :!event Custom|HandleLanded) :!class /SonicSpearGameplay/Gameplay/Drops/BGA_SonicSpear_MainDropParent.BGA_SonicSpear_MainDropParent_C))
                (:CastFailed
                  (Fortnite|FortShippingLog self (Utilities|String|Append "BeginPlay:  Expected GetOwner to return BGA_SonicSpear_MainDropParent, but instead got: " (Utilities|GetDisplayName (Actor|GetOwner))) true))))
            (:Is%20Not%20Valid
              (Fortnite|FortShippingLog self "BeginPlay:  GetOwner is null" true))))))
    (:then_1
      (Utilities|FlowControl|Delay 0.5)
      (Rendering|SetRenderCustomDepth (Variables|Debug|GetStaticMeshComponent) false))
    (:then_2
      ;;[1 1 1] Use Hotfixable Value for Auto-Ping Trigger Distance
      (Components|Sphere|SetSphereRadius (Variables|Default|GetSphereCollisionAutoPing) (ScalableFloat|GetValueatLevel (Variables|AutoPing|GetAutoPingTriggerDistance)) false)
      ;;[1 1 1] Use Hotfixable Value for "Destroy" Chevron Trigger Distance
      (Components|Sphere|SetSphereRadius (Variables|Default|GetSphereCollisionDestroyMMOMarkerTrigger) (ScalableFloat|GetValueatLevel (Variables|MMOMarker|GetDestroyMMOMarkerTriggerDistance)) false))))

(event Custom|HandleLanded
  (bind (_returnvalue_location _returnvalue_rotation) (Math|Transform|BreakTransform (Transformation|GetActorTransform)))
  ;[1 1 0 0.5] [OnLanded]
  (Utilities|FlowControl|DoOnce false
    (:Completed
      (Networking|SwitchHasAuthority
        (:Authority
          (for _array_element (Variables|Default|GetPointProviderRelativeTransforms)
            (bind (_array_element_location _array_element_rotation) (Math|Transform|BreakTransform _array_element))
            (bind _returnvalue (Game|SpawnActorfromClass "/BuzzCool/Gameplay/BP_BuzzCool_CloneSpawnActor_StaticPointProvider.BP_BuzzCool_CloneSpawnActor_StaticPointProvider_C" (Math|Transform|ComposeTransforms (Math|Transform|MakeTransform _array_element_location _array_element_rotation) (Math|Transform|MakeTransform _returnvalue_location _returnvalue_rotation)) "AlwaysSpawn" "OverrideRootScale" self 0))
            (Utilities|Array|Add (Variables|Internal|GetLivingWorld_SpawnLocations) _returnvalue)
            (if (Variables|Default|GetbShouldAttachPointProviders)
              (Transformation|AttachActorToComponent _returnvalue (Variables|Debug|GetStaticMeshComponent) "None" "KeepWorld" "KeepWorld" "KeepWorld" false))))))))

(event EventOnDeathPlayEffects
  (params (Damage :type Float%20%28single-precision%29) (DamageTags :type Gameplay%20Tag%20Container%20Structure%20%28by%20ref%29) (Momentum :type Vector :default "0, 0, 0") (HitInfo :type Hit%20Result%20Structure%20%28by%20ref%29) (InstigatedBy :type Fort%20Pawn%20Object%20Reference) (DamageCauser :type Actor%20Object%20Reference) (EffectContext :type Gameplay%20Effect%20Context%20Handle%20Structure))
  (bind _coloring_frostingcoloring_15_ed6d36ac450de5d5c052ada0a0a175be (Utilities|Struct|BreakSSonicSpearDonutDropColor (Variables|GetColoring_FrostingColoring_15_ED6D36AC450DE5D5C052ADA0A0A175BE)))
  ;[1 0.5 0 0.5] [OnDeathPlayEffects]
  (bind _returnvalue (Niagara|SpawnSystematLocation "/BuzzCool/Assets/FX/NS_BuzzCool_Donut_Explosion.NS_BuzzCool_Donut_Explosion" (Transformation|GetWorldLocation (Variables|Debug|GetStaticMeshComponent)) (Transformation|GetWorldRotation (Variables|Debug|GetStaticMeshComponent))))
  (Effects|Components|ParticleSystem|SetColorParameter _returnvalue "FrostingColor" _coloring_frostingcoloring_15_ed6d36ac450de5d5c052ada0a0a175be)
  (Audio|PlaySoundatLocation (Variables|Audio|GetExplodeSound) (Transformation|GetActorLocation))
  (bind _audioloop (Variables|Default|GetAudioLoop)
    (:then
      (Audio|Components|Audio|FadeOut _audioloop 0.5 0.0 "Linear" :!class /Script/Engine.AudioComponent))
    (:else)))

(event EventEndPlay
  (params (EndPlayReason :type EEndPlayReason%20Enum :default "Destroyed"))
  (bind _returnvalue (Actor|GetOwner))
  (bind _livingworld_spawnlocations (Variables|Internal|GetLivingWorld_SpawnLocations))
  ;[1 0 0 0.5] [EndPlay]
  (|CallParentFunction EndPlayReason)
  (Networking|SwitchHasAuthority
    (:Authority
      (Utilities|FlowControl|Sequence
        (:then_0
          (Utilities|IsValid _returnvalue
            (:Is%20Valid
              (bind _asbga_sonic_spear_main_drop_parent (Utilities|Casting|CastToBGA_SonicSpear_MainDropParent _returnvalue))
              (Default|UnbindEventfromOnLanded _asbga_sonic_spear_main_drop_parent (EventDispatchers|CreateEvent :!event Custom|HandleLanded) :!class /SonicSpearGameplay/Gameplay/Drops/BGA_SonicSpear_MainDropParent.BGA_SonicSpear_MainDropParent_C))))
        (:then_1
          (Utilities|Time|ClearandInvalidateTimerbyHandle (Variables|Default|GetDebugPointProvidersTimer))
          (for _array_element _livingworld_spawnlocations
            (Utilities|IsValid _array_element
              (:Is%20Valid
                (Actor|DestroyActor _array_element))))
          (Utilities|Array|Clear _livingworld_spawnlocations))))))

(event Custom|BuzzCool_DebugPointProviders
  ;[0.15 0.000307 0.1473 0.5] [BuzzCool_DebugPointProviders]  In PIE, use "cheat ke* BuzzCool_DebugPointProviders"
  (Networking|SwitchHasAuthority
    (:Authority
      (if (Utilities|Time|IsTimerActivebyHandle (Variables|Default|GetDebugPointProvidersTimer))
        (Utilities|Time|ClearandInvalidateTimerbyHandle (Variables|Default|GetDebugPointProvidersTimer))
        (else
          (bind _returnvalue (Utilities|Time|SetTimerbyEvent (EventDispatchers|CreateEvent :!event Custom|DebugDrawPointProviders) 0.01 true true))
          (Variables|Default|SetDebugPointProvidersTimer _returnvalue))))))

(event Custom|DebugDrawPointProviders
  ;[0.15 0.000307 0.1473 0.5] [BuzzCool_DebugPointProviders]  In PIE, use "cheat ke* BuzzCool_DebugPointProviders"
  (for _array_element (Variables|Internal|GetLivingWorld_SpawnLocations)
    (Utilities|IsValid _array_element
      (:Is%20Valid
        (Rendering|Debug|DrawDebugArrow (Transformation|GetActorLocation _array_element) (+ (Transformation|GetActorLocation _array_element) (* (Transformation|GetActorForwardVector _array_element) 2e+02)) 5e+02 "(R=1.000000,G=0.000000,B=0.000000,A=1.000000)" 0.0 5.0)))))

(event OnComponentBeginOverlap(SphereCollisionAutoPing)
  (params (OverlappedComponent :type Primitive%20Component%20Object%20Reference) (OtherActor :type Actor%20Object%20Reference) (OtherComp :type Primitive%20Component%20Object%20Reference) (OtherBodyIndex :type Integer) (bFromSweep :type Boolean) (SweepResult :type Hit%20Result%20Structure%20%28by%20ref%29))
  ;[1 1 1] Auto-Ping Players Entering Trigger Distance
  (Networking|CanExecuteCosmeticEvents
    (:True
      (bind _asfort_player_pawn_athena (Utilities|Casting|CastToFortPlayerPawnAthena OtherActor))
      (if (Utilities|IsValid (Fortnite|GetFortPlayerControllerfromActor _asfort_player_pawn_athena) :!class /Script/Engine.KismetSystemLibrary)
        (if (Pawn|IsLocalPlayerController (Fortnite|GetFortPlayerControllerfromActor _asfort_player_pawn_athena))
          (GameplayEventRouter|BroadcastEvent (GameplayEventRouter|GetGlobalEventRouter) (Utilities|Struct|MakeGameplayEventData "(TagName=\"BuzzCool.Gameplay.EnteredAutoPingArea\")" self (Fortnite|GetFortPlayerControllerfromActor _asfort_player_pawn_athena)) (Fortnite|GetFortPlayerControllerfromActor _asfort_player_pawn_athena)))))) ; Am I a client?
)

(event OnComponentBeginOverlap(SphereCollisionDestroyMMOMarkerTrigger)
  (params (OverlappedComponent :type Primitive%20Component%20Object%20Reference) (OtherActor :type Actor%20Object%20Reference) (OtherComp :type Primitive%20Component%20Object%20Reference) (OtherBodyIndex :type Integer) (bFromSweep :type Boolean) (SweepResult :type Hit%20Result%20Structure%20%28by%20ref%29))
  ;[1 1 1] Show "Destroy" MMO Marker
  (Networking|CanExecuteCosmeticEvents
    (:True
      (bind _asfort_player_pawn_athena (Utilities|Casting|CastToFortPlayerPawnAthena OtherActor))
      (if (Utilities|IsValid (Fortnite|GetFortPlayerControllerfromActor _asfort_player_pawn_athena) :!class /Script/Engine.KismetSystemLibrary)
        (if (Pawn|IsLocalPlayerController (Fortnite|GetFortPlayerControllerfromActor _asfort_player_pawn_athena))
          (GameplayEventRouter|BroadcastEvent (GameplayEventRouter|GetGlobalEventRouter) (Utilities|Struct|MakeGameplayEventData "(TagName=\"BuzzCool.Gameplay.EnteredObjectiveMarkerArea\")" self (Fortnite|GetFortPlayerControllerfromActor _asfort_player_pawn_athena)) (Fortnite|GetFortPlayerControllerfromActor _asfort_player_pawn_athena)))))) ; Am I a client?
)

(event OnComponentEndOverlap(SphereCollisionDestroyMMOMarkerTrigger)
  (params (OverlappedComponent :type Primitive%20Component%20Object%20Reference) (OtherActor :type Actor%20Object%20Reference) (OtherComp :type Primitive%20Component%20Object%20Reference) (OtherBodyIndex :type Integer))
  ;[1 1 1] Hide "Destroy" MMO Marker
  (Networking|CanExecuteCosmeticEvents
    (:True
      (bind _asfort_player_pawn_athena (Utilities|Casting|CastToFortPlayerPawnAthena OtherActor))
      (if (Utilities|IsValid (Fortnite|GetFortPlayerControllerfromActor _asfort_player_pawn_athena) :!class /Script/Engine.KismetSystemLibrary)
        (if (Pawn|IsLocalPlayerController (Fortnite|GetFortPlayerControllerfromActor _asfort_player_pawn_athena))
          (GameplayEventRouter|BroadcastEvent (GameplayEventRouter|GetGlobalEventRouter) (Utilities|Struct|MakeGameplayEventData "(TagName=\"BuzzCool.Gameplay.ExitObjectiveMarkerArea\")" self (Fortnite|GetFortPlayerControllerfromActor _asfort_player_pawn_athena)) (Fortnite|GetFortPlayerControllerfromActor _asfort_player_pawn_athena)))))) ; Am I a client?
)

(fn GetSpawnActorProcessor
  (return (Variables|Internal|GetCloneSpawnActorProcessor)))

(fn ConstructionScript)
