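; BuzzCool thrown projectile that deploys a shield bubble on impact through its parent class's spawn logic.
; On authority, hits on existing shield bubbles are ignored; other hits spawn only outside deployable-block volumes, and the attach-to-hit-actor function is a stub returning false.
; Remote hits show a temporary Niagara splash and randomized decal, and every hit fades the in-air audio.
; Bounces off shield bubbles restore pre-bounce velocity a tick later, after the movement component finishes updating.
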(event EventBeginPlay
  (Networking|SwitchHasAuthority))

(event Collision|EventHit
  (params (MyComp :type Primitive%20Component%20Object%20Reference) (Other :type Actor%20Object%20Reference) (OtherComp :type Primitive%20Component%20Object%20Reference) (bSelfMoved :type Boolean) (HitLocation :type Vector :default "0, 0, 0") (HitNormal :type Vector :default "0, 0, 0") (NormalImpulse :type Vector :default "0, 0, 0") (Hit :type Hit%20Result%20Structure%20%28by%20ref%29))
  (bind (_bblockinghit _binitialoverlap _time _distance _location _impactpoint _normal) (Collision|BreakHitResult Hit))
  ;[0 1 0.9019 0.484] Event Hit()
  (Utilities|FlowControl|Sequence
    (:then_0
      (Networking|SwitchHasAuthority
        (:Authority
          (Utilities|IsAShieldBubbleSafeZone Other
            (:Then
              ;;[1 0.4844 0.4844] If we hit a shield, add it to the ignore list
              (Collision|IgnoreActorwhenMoving (Variables|Collision|GetCapsuleComponent) Other true))
            (:Cast%20Failed
              (bind _boverlapping (Default|IsOverlappingDeployableBlockVolume self (Variables|Spawning|ValidityChecks|GetBlockingAreaCheckRadius) (Transformation|GetActorLocation) (Variables|Spawning|GetItem%20Specific%20Blocking%20Tag)))
              (if (not _boverlapping)
                ;;[0 0.1464 0.4531] Spawn and attach the shield bubble
                (|CallParentFunction MyComp Other OtherComp bSelfMoved HitLocation HitNormal NormalImpulse Hit) ; runs additional validity checks and\nspawns the bubble shield (if possible)
                (CallFunction|AttempttoAttachtoHitActor (Variables|Spawning|GetSpawnedBGA) Hit)))))
        (:Remote
          ;[1 1 1] Temp donut splash effect
          (Niagara|SpawnSystematLocation "/BuzzCool/Assets/FX/NS_BuzzCool_Radial_Impact.NS_BuzzCool_Radial_Impact" _location (Math|Conversions|RotationFromXVector _normal))
          (CallFunction|SpawnDecal "NewEnumerator0" _normal _location))))
    (:then_1
      ;;[1 1 1] Stop audio
      (bind _inairloop (Variables|Default|GetInAirLoop)
        (:then
          (Audio|Components|Audio|FadeOut _inairloop 0.25 0.0 "Linear" :!class /Script/Engine.AudioComponent))
        (:else)))))

(event OnProjectileBounce(ProjectileMovementComponent)
  (params (ImpactResult :type Hit%20Result%20Structure%20%28by%20ref%29) (ImpactVelocity :type Vector%20%28by%20ref%29 :default "0, 0, 0"))
  (bind (_bblockinghit _binitialoverlap _time _distance _location _impactpoint _normal _impactnormal _physmat _hitactor) (Collision|BreakHitResult ImpactResult))
  ;[0.2937 0.562 1 0.524] OnProjectileBounce()
  ; If we hit a shield we want to continue moving through it without bouncing
  ; Immediately add the actor to our ignore list and then set our velocity to the pre-bounce velocity.
  (Utilities|IsAShieldBubbleSafeZone _hitactor
    (:Then
      (Utilities|FlowControl|DelayUntilNextTick) ; We need to wait 1 frame because the Projectile Movement Component is not done updating the velocity for this frame.
      (Projectile|SetVelocity (Variables|Movement|GetProjectileMovementComponent) ImpactVelocity))))

(fn AttemptToAttachToHitActor
  (params (SpawnedBubble :type Actor%20Object%20Reference) (HitResult :type Hit%20Result%20Structure))
  ;[0.9237 0.419 1] Return No Valid Attachment
  (return false))

(fn GetDecalMaterial
  (return (Utilities|Array|Get(acopy) (Variables|ProtoSplash|GetDecalMaterials) (Math|Random|RandomIntegerinRange 0 (Utilities|Array|LastIndex (Variables|ProtoSplash|GetDecalMaterials))))))

(fn SpawnBGA_
  (params (Normal :type Vector :default "0, 0, 0") (HitActor :type Actor%20Object%20Reference) (HitComp :type Primitive%20Component%20Object%20Reference) (Pos :type Vector :default "0, 0, 0") (FromWater :type Boolean) (Hit :type Hit%20Result%20Structure))
  (bind _boverlapping (Default|IsOverlappingDeployableBlockVolume self (Variables|Spawning|ValidityChecks|GetBlockingAreaCheckRadius) (Transformation|GetActorLocation) (Variables|Spawning|GetItem%20Specific%20Blocking%20Tag)))
  (if (not _boverlapping)
    (bind _returnvalue (|CallParentFunction Normal HitActor HitComp Pos FromWater Hit))
    (Default|CallEventSpawnedBGA _returnvalue)
    (return _returnvalue)))

(fn SpawnDecal
  (params (CanSpawn :type En_Grenade_Papaya_ActorTypes%20Enum :default "NewEnumerator0") (Normal :type Vector) (Location :type Vector))
  (bind _output (CallFunction|GetDecalMaterial))
  (Utilities|IsValid _output
    (:Is%20Valid
      (bind _returnvalue (Rendering|Decal|SpawnDecalatLocation _output (Math|Vector|MakeVector (Utilities|Select :Index (Variables|ProtoSplash|GetDecalActorType) :NewEnumerator0 (Variables|ProtoSplash|GetDecalDepth) :NewEnumerator1 0.0 :NewEnumerator2 (Variables|ProtoSplash|GetDecalDepthLandscape)) (Math|Random|RandomFloatinRange (Variables|ProtoSplash|GetDecalSizeMin) (Variables|ProtoSplash|GetDecalSizeMax)) (Math|Random|RandomFloatinRange (Variables|ProtoSplash|GetDecalSizeMin) (Variables|ProtoSplash|GetDecalSizeMax))) Location (Math|Rotator|MakeRotfromX (* Normal -1.0)) (Variables|ProtoSplash|GetDecalLifetime)))
      (Transformation|AddLocalRotation _returnvalue (Math|Rotator|MakeRotator (Math|Random|RandomFloatinRange 0.0 3.6e+02)) false false)
      (Rendering|Components|Decal|SetFadeIn _returnvalue 0.0 (Variables|ProtoSplash|GetDecalFadeInDuration))
      (Rendering|Components|Decal|SetFadeOut _returnvalue (Variables|ProtoSplash|GetDecalDelayBeforeFade) (Variables|ProtoSplash|GetDecalFadeOutDuration) true))))

(fn ConstructionScript
  (|CallParentFunction))
