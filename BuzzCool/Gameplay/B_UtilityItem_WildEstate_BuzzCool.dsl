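; Blueprint for the WildEstate utility item weapon in BuzzCool, customizing rendering when it is equipped.
; On the equipped-and-attached event it runs the parent logic, waits one tick, then sets a custom depth stencil value on the weapon mesh.
; The construction script only calls the parent.
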(event FortWeapon|EventOnEquippedAndAttached
  (|CallParentFunction)
  (Utilities|FlowControl|DelayUntilNextTick)
  (Rendering|SetCustomDepthStencilValue (Variables|Visual|GetWeaponMesh) 5))

(fn ConstructionScript
  (|CallParentFunction))
